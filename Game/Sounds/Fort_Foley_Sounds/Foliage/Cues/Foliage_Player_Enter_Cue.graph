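# Foley sound cue for a player entering foliage, playing one of seven equally weighted waves with slight modulation.
# An IsLocal branch sends the local player through its own attenuation node and others through the shared third-person footstep attenuation.
# For non-local players, an IsTeam branch plays teammates much quieter through an extra fixed-volume modulator.
# Concurrency override limits the cue to one instance per owner.
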SoundCue /Game/Sounds/Fort_Foley_Sounds/Foliage/Cues/Foliage_Player_Enter_Cue
root=Branch_4
nodes [
 {id=Attenuation_0, children [Modulator_0]}
 {id=Attenuation_1, children [Branch_5], properties {attenuationSettings=/Game/Sounds/Fort_Player/Footsteps/Settings/Fort_Player_FS_3P_Att}}
 {id=Branch_4, children [Attenuation_0, Attenuation_1, Attenuation_1], properties {boolParameterName=IsLocal}}
 {id=Branch_5, children [Modulator_3,Modulator_0,Modulator_0], properties {boolParameterName=IsTeam}}
 {id=Modulator_0, children [Random_0], properties {volumeMin=0.8}}
 {id=Modulator_3, children [Modulator_0], properties {pitchMax=1, pitchMin=1, volumeMax=0.2, volumeMin=0.2}}
 {
  id=Random_0
  children [WavePlayer_0,WavePlayer_1,WavePlayer_2,WavePlayer_3,WavePlayer_4,WavePlayer_5,WavePlayer_6]
  properties {bShouldExcludeFromBranchCulling=true, weights [1,1,1,1,1,1,1]}
 }
 {id=WavePlayer_0, properties {soundWaveAssetPtr=/Game/Sounds/Fort_Foley_Sounds/Foliage/Foliage_Player_Enter_01}}
 {id=WavePlayer_1, properties {soundWaveAssetPtr=/Game/Sounds/Fort_Foley_Sounds/Foliage/Foliage_Player_Enter_02}}
 {id=WavePlayer_2, properties {soundWaveAssetPtr=/Game/Sounds/Fort_Foley_Sounds/Foliage/Foliage_Player_Enter_03}}
 {id=WavePlayer_3, properties {soundWaveAssetPtr=/Game/Sounds/Fort_Foley_Sounds/Foliage/Foliage_Player_Enter_04}}
 {id=WavePlayer_4, properties {soundWaveAssetPtr=/Game/Sounds/Fort_Foley_Sounds/Foliage/Foliage_Player_Enter_05}}
 {id=WavePlayer_5, properties {soundWaveAssetPtr=/Game/Sounds/Fort_Foley_Sounds/Foliage/Foliage_Player_Enter_06}}
 {id=WavePlayer_6, properties {soundWaveAssetPtr=/Game/Sounds/Fort_Foley_Sounds/Foliage/Foliage_Player_Enter_07}}
]
properties {
 bOverrideConcurrency=true
 maxDistance=2097152
 soundClassObject=&Fort_Player_Foley
 volumeMultiplier=1.25
 concurrencyOverrides {bLimitToOwner=true, maxCount=1, volumeScaleAttackTime=0.2}
 concurrencySet [null,null]
}
