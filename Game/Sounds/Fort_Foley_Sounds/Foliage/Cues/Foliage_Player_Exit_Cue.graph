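# Foley cue for a player exiting foliage, playing one of six equally weighted exit waves with slight volume variation.
# Nested IsLocal and IsTeam branches share that random pool, with an extra modulator cutting volume to a fifth for non-local teammates only.
# It uses third-person footstep attenuation and the player foley sound class, with a concurrency override limiting it to one instance per owner.
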SoundCue /Game/Sounds/Fort_Foley_Sounds/Foliage/Cues/Foliage_Player_Exit_Cue
root=Branch_4
nodes [
 {id=Branch_4, children [Modulator_0,Branch_5,Branch_5], properties {boolParameterName=IsLocal}}
 {id=Branch_5, children [Modulator_3,Modulator_0,Modulator_0], properties {boolParameterName=IsTeam}}
 {id=Modulator_0, children [Random_1], properties {volumeMin=0.8}}
 {id=Modulator_3, children [Modulator_0], properties {pitchMax=1, pitchMin=1, volumeMax=0.2, volumeMin=0.2}}
 {
  id=Random_1
  children [WavePlayer_7, WavePlayer_8, WavePlayer_9, WavePlayer_10, WavePlayer_11, WavePlayer_12]
  properties {bShouldExcludeFromBranchCulling=true, weights [1,1,1,1,1,1]}
 }
 {id=WavePlayer_10, properties {soundWaveAssetPtr=/Game/Sounds/Fort_Foley_Sounds/Foliage/Foliage_Player_Exit_03}}
 {id=WavePlayer_11, properties {soundWaveAssetPtr=/Game/Sounds/Fort_Foley_Sounds/Foliage/Foliage_Player_Exit_04}}
 {id=WavePlayer_12, properties {soundWaveAssetPtr=/Game/Sounds/Fort_Foley_Sounds/Foliage/Foliage_Player_Exit_05}}
 {id=WavePlayer_7, properties {soundWaveAssetPtr=/Game/Sounds/Fort_Foley_Sounds/Foliage/Foliage_Player_Exit_06}}
 {id=WavePlayer_8, properties {soundWaveAssetPtr=/Game/Sounds/Fort_Foley_Sounds/Foliage/Foliage_Player_Exit_01}}
 {id=WavePlayer_9, properties {soundWaveAssetPtr=/Game/Sounds/Fort_Foley_Sounds/Foliage/Foliage_Player_Exit_02}}
]
properties {
 attenuationSettings=/Game/Sounds/Fort_Player/Footsteps/Settings/Fort_Player_FS_3P_Att
 bOverrideConcurrency=true
 maxDistance=3584
 soundClassObject=&Fort_Player_Foley
 volumeMultiplier=1.25
 concurrencyOverrides {bLimitToOwner=true, maxCount=1, volumeScaleAttackTime=0.2}
 concurrencySet [null,null]
}
