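# Sound cue for soft, long player movement through foliage, with concurrency settings limiting simultaneous instances.
# IsLocal picks first-person attenuation and sound class; others use third-person attenuation and the enemy class, with IsTeam teammates cut to low volume.
# All paths share a quality switch between larger and smaller wave pools.
# Each pool mixes random soft-long movement and additional-movement layers, varied by volume and pitch modulators.
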SoundCue /Game/Sounds/Fort_Foley_Sounds/Foliage/Cues/Foliage_Player_Movement_Soft_Long_Cue
root=Branch_4
nodes [
 {
  id=Attenuation_4
  children [SoundClass_4]
  properties {attenuationSettings=/Game/Sounds/Fort_Player/Footsteps/Settings/Fort_Player_FS_1P_Att}
 }
 {
  id=Attenuation_5
  children [SoundClass_5]
  properties {attenuationSettings=/Game/Sounds/Fort_Player/Footsteps/Settings/Fort_Player_FS_3P_Att}
 }
 {id=Branch_4, children [Attenuation_4, Branch_5, Branch_5], properties {boolParameterName=IsLocal}}
 {id=Branch_5, children [Modulator_3, Attenuation_5, Attenuation_5], properties {boolParameterName=IsTeam}}
 {id=Mixer_1, children [Modulator_2,Modulator_5], properties {inputVolume [1,1]}}
 {id=Mixer_2, children [Modulator_6,Modulator_4], properties {inputVolume [1,1]}}
 {id=Modulator_0, children [QualityLevel_0], properties {volumeMin=0.8}}
 {id=Modulator_2, children [Random_6], properties {pitchMin=0.8, volumeMax=1, volumeMin=0.5}}
 {id=Modulator_3, children [Attenuation_5], properties {pitchMax=1, pitchMin=1, volumeMax=0.2, volumeMin=0.2}}
 {id=Modulator_4, children [Random_7], properties {pitchMin=0.8, volumeMax=1, volumeMin=0.5}}
 {id=Modulator_5, children [Random_3], properties {pitchMin=0.8, volumeMax=1, volumeMin=0.3}}
 {id=Modulator_6, children [Random_4], properties {pitchMin=0.8, volumeMax=1, volumeMin=0.3}}
 {id=QualityLevel_0, children [Mixer_1,Mixer_1,Mixer_2,Mixer_2]}
 {
  id=Random_3
  children [
   WavePlayer_21
   WavePlayer_22
   WavePlayer_23
   WavePlayer_24
   WavePlayer_25
   WavePlayer_26
   WavePlayer_27
   WavePlayer_28
   WavePlayer_29
   WavePlayer_30
  ]
  properties {weights [1,1,1,1,1,1,1,1,1,1]}
 }
 {
  id=Random_4
  children [WavePlayer_31, WavePlayer_32, WavePlayer_33, WavePlayer_34, WavePlayer_35]
  properties {bShouldExcludeFromBranchCulling=true, weights [1,1,1,1,1]}
 }
 {
  id=Random_6
  children [
   WavePlayer_17
   WavePlayer_18
   WavePlayer_19
   WavePlayer_20
   WavePlayer_37
   WavePlayer_38
   WavePlayer_39
   WavePlayer_43
   WavePlayer_44
   WavePlayer_45
   WavePlayer_46
   WavePlayer_47
   WavePlayer_48
   WavePlayer_49
  ]
  properties {weights [1,1,1,1,1,1,1,1,1,1,1,1,1,1]}
 }
 {
  id=Random_7
  children [WavePlayer_16, WavePlayer_36, WavePlayer_40, WavePlayer_42, WavePlayer_41]
  properties {bShouldExcludeFromBranchCulling=true, weights [1,1,1,1,1]}
 }
 {id=SoundClass_4, children [Modulator_0], properties {soundClassOverride=&Fort_Player_Foley_1P_Athena}}
 {
  id=SoundClass_5
  children [Modulator_0]
  properties {soundClassOverride=&Fort_Player_Foley_3P_Enemy_Athena}
 }
 {id=WavePlayer_16, properties {soundWaveAssetPtr=/Game/Sounds/Fort_Foley_Sounds/Foliage/Foliage_AdditionalMovements_04}}
 {id=WavePlayer_17, properties {soundWaveAssetPtr=/Game/Sounds/Fort_Foley_Sounds/Foliage/Foliage_AdditionalMovements_01}}
 {id=WavePlayer_18, properties {soundWaveAssetPtr=&Foliage_AdditionalMovements_04}}
 {id=WavePlayer_19, properties {soundWaveAssetPtr=/Game/Sounds/Fort_Foley_Sounds/Foliage/Foliage_AdditionalMovements_05}}
 {id=WavePlayer_20, properties {soundWaveAssetPtr=/Game/Sounds/Fort_Foley_Sounds/Foliage/Foliage_AdditionalMovements_06}}
 {id=WavePlayer_21, properties {soundWaveAssetPtr=/Game/Sounds/Fort_Foley_Sounds/Foliage/Foliage_Player_Movement_Soft_Long_01}}
 {id=WavePlayer_22, properties {soundWaveAssetPtr=/Game/Sounds/Fort_Foley_Sounds/Foliage/Foliage_Player_Movement_Soft_Long_02}}
 {id=WavePlayer_23, properties {soundWaveAssetPtr=/Game/Sounds/Fort_Foley_Sounds/Foliage/Foliage_Player_Movement_Soft_Long_03}}
 {id=WavePlayer_24, properties {soundWaveAssetPtr=/Game/Sounds/Fort_Foley_Sounds/Foliage/Foliage_Player_Movement_Soft_Long_04}}
 {id=WavePlayer_25, properties {soundWaveAssetPtr=/Game/Sounds/Fort_Foley_Sounds/Foliage/Foliage_Player_Movement_Soft_Long_05}}
 {id=WavePlayer_26, properties {soundWaveAssetPtr=/Game/Sounds/Fort_Foley_Sounds/Foliage/Foliage_Player_Movement_Soft_Long_06}}
 {id=WavePlayer_27, properties {soundWaveAssetPtr=/Game/Sounds/Fort_Foley_Sounds/Foliage/Foliage_Player_Movement_Soft_Long_07}}
 {id=WavePlayer_28, properties {soundWaveAssetPtr=/Game/Sounds/Fort_Foley_Sounds/Foliage/Foliage_Player_Movement_Soft_Long_08}}
 {id=WavePlayer_29, properties {soundWaveAssetPtr=/Game/Sounds/Fort_Foley_Sounds/Foliage/Foliage_Player_Movement_Soft_Long_09}}
 {id=WavePlayer_30, properties {soundWaveAssetPtr=/Game/Sounds/Fort_Foley_Sounds/Foliage/Foliage_Player_Movement_Soft_Long_10}}
 {id=WavePlayer_31, properties {soundWaveAssetPtr=&Foliage_Player_Movement_Soft_Long_01}}
 {id=WavePlayer_32, properties {soundWaveAssetPtr=&Foliage_Player_Movement_Soft_Long_02}}
 {id=WavePlayer_33, properties {soundWaveAssetPtr=&Foliage_Player_Movement_Soft_Long_03}}
 {id=WavePlayer_34, properties {soundWaveAssetPtr=&Foliage_Player_Movement_Soft_Long_04}}
 {id=WavePlayer_35, properties {soundWaveAssetPtr=&Foliage_Player_Movement_Soft_Long_05}}
 {id=WavePlayer_36, properties {soundWaveAssetPtr=&Foliage_AdditionalMovements_06}}
 {id=WavePlayer_37, properties {soundWaveAssetPtr=/Game/Sounds/Fort_Foley_Sounds/Foliage/Foliage_AdditionalMovements_08}}
 {id=WavePlayer_38, properties {soundWaveAssetPtr=/Game/Sounds/Fort_Foley_Sounds/Foliage/Foliage_AdditionalMovements_09}}
 {id=WavePlayer_39, properties {soundWaveAssetPtr=/Game/Sounds/Fort_Foley_Sounds/Foliage/Foliage_AdditionalMovements_11}}
 {id=WavePlayer_40, properties {soundWaveAssetPtr=/Game/Sounds/Fort_Foley_Sounds/Foliage/Foliage_AdditionalMovements_14}}
 {id=WavePlayer_41, properties {soundWaveAssetPtr=/Game/Sounds/Fort_Foley_Sounds/Foliage/Foliage_AdditionalMovements_15}}
 {id=WavePlayer_42, properties {soundWaveAssetPtr=/Game/Sounds/Fort_Foley_Sounds/Foliage/Foliage_AdditionalMovements_16}}
 {id=WavePlayer_43, properties {soundWaveAssetPtr=/Game/Sounds/Fort_Foley_Sounds/Foliage/Foliage_AdditionalMovements_12}}
 {id=WavePlayer_44, properties {soundWaveAssetPtr=/Game/Sounds/Fort_Foley_Sounds/Foliage/Foliage_AdditionalMovements_13}}
 {id=WavePlayer_45, properties {soundWaveAssetPtr=&Foliage_AdditionalMovements_14}}
 {id=WavePlayer_46, properties {soundWaveAssetPtr=&Foliage_AdditionalMovements_15}}
 {id=WavePlayer_47, properties {soundWaveAssetPtr=&Foliage_AdditionalMovements_16}}
 {id=WavePlayer_48, properties {soundWaveAssetPtr=/Game/Sounds/Fort_Foley_Sounds/Foliage/Foliage_AdditionalMovements_17}}
 {id=WavePlayer_49, properties {soundWaveAssetPtr=/Game/Sounds/Fort_Foley_Sounds/Foliage/Foliage_AdditionalMovements_18}}
]
properties {
 maxDistance=3584
 volumeMultiplier=1
 concurrencyOverrides {bLimitToOwner=true, maxCount=2, resolutionRule=StopOldest, volumeScaleAttackTime=0.2}
 concurrencySet [
  /Game/Sounds/Fort_Foley_Sounds/Foliage/Settings/Athena_Foley_Foliage_Local_Conc
  /Game/Sounds/Fort_Foley_Sounds/Foliage/Settings/Athena_Foley_Foliage_Global_Conc
 ]
}
