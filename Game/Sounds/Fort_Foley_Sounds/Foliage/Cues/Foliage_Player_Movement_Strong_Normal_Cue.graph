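# Sound cue for a player's strong normal movement through foliage, mixed by listener relationship.
# IsLocal selects first-person attenuation and sound class; otherwise third-person enemy settings apply, with IsTeam cutting teammates' volume sharply.
# All paths share a quality switch between two mixers, each layering random movement and additional-movement waves with pitch and volume variation.
# The last two quality levels use smaller wave pools excluded from branch culling, and concurrency rules limit overlap.
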SoundCue /Game/Sounds/Fort_Foley_Sounds/Foliage/Cues/Foliage_Player_Movement_Strong_Normal_Cue
root=Branch_4
nodes [
 {
  id=Attenuation_4
  children [SoundClass_4]
  properties {attenuationSettings=/Game/Sounds/Fort_Player/Footsteps/Settings/Fort_Player_FS_1P_Att}
 }
 {
  id=Attenuation_5
  children [SoundClass_5]
  properties {attenuationSettings=/Game/Sounds/Fort_Player/Footsteps/Settings/Fort_Player_FS_3P_Att}
 }
 {id=Branch_4, children [Attenuation_4, Branch_5, Branch_5], properties {boolParameterName=IsLocal}}
 {id=Branch_5, children [Modulator_3, Attenuation_5, Attenuation_5], properties {boolParameterName=IsTeam}}
 {id=Mixer_0, children [Modulator_2,Modulator_5], properties {inputVolume [1,1]}}
 {id=Mixer_1, children [Modulator_1,Modulator_4], properties {inputVolume [1,1]}}
 {id=Modulator_0, children [QualityLevel_0], properties {volumeMin=0.8}}
 {id=Modulator_1, children [Random_2], properties {pitchMin=0.8, volumeMax=1.3, volumeMin=0.4}}
 {id=Modulator_2, children [Random_4], properties {pitchMin=0.8, volumeMax=1.3, volumeMin=0.6}}
 {id=Modulator_3, children [Attenuation_5], properties {pitchMax=1, pitchMin=1, volumeMax=0.2, volumeMin=0.2}}
 {id=Modulator_4, children [Random_5], properties {pitchMin=0.8, volumeMax=1.3, volumeMin=0.6}}
 {id=Modulator_5, children [Random_0], properties {pitchMin=0.8, volumeMax=1.3, volumeMin=0.4}}
 {id=QualityLevel_0, children [Mixer_0,Mixer_0,Mixer_1,Mixer_1]}
 {
  id=Random_0
  children [
   WavePlayer_0
   WavePlayer_1
   WavePlayer_2
   WavePlayer_3
   WavePlayer_4
   WavePlayer_5
   WavePlayer_6
   WavePlayer_13
   WavePlayer_14
   WavePlayer_15
  ]
  properties {weights [1,1,1,1,1,1,1,1,1,1]}
 }
 {
  id=Random_2
  children [WavePlayer_16, WavePlayer_17, WavePlayer_18, WavePlayer_19, WavePlayer_20]
  properties {bShouldExcludeFromBranchCulling=true, weights [1,1,1,1,1]}
 }
 {
  id=Random_4
  children [
   WavePlayer_7
   WavePlayer_10
   WavePlayer_11
   WavePlayer_12
   WavePlayer_22
   WavePlayer_23
   WavePlayer_25
   WavePlayer_26
   WavePlayer_27
   WavePlayer_28
   WavePlayer_29
   WavePlayer_30
   WavePlayer_31
   WavePlayer_32
  ]
  properties {weights [1,1,1,1,1,1,1,1,1,1,1,1,1,1]}
 }
 {
  id=Random_5
  children [WavePlayer_33, WavePlayer_34, WavePlayer_35, WavePlayer_36, WavePlayer_37]
  properties {bShouldExcludeFromBranchCulling=true, weights [1,1,1,1,1]}
 }
 {id=SoundClass_4, children [Modulator_0], properties {soundClassOverride=&Fort_Player_Foley_1P_Athena}}
 {
  id=SoundClass_5
  children [Modulator_0]
  properties {soundClassOverride=&Fort_Player_Foley_3P_Enemy_Athena}
 }
 {id=WavePlayer_0, properties {soundWaveAssetPtr=/Game/Sounds/Fort_Foley_Sounds/Foliage/Foliage_Player_Movement_Strong_Normal_01}}
 {id=WavePlayer_1, properties {soundWaveAssetPtr=/Game/Sounds/Fort_Foley_Sounds/Foliage/Foliage_Player_Movement_Strong_Normal_02}}
 {id=WavePlayer_10, properties {soundWaveAssetPtr=/Game/Sounds/Fort_Foley_Sounds/Foliage/Foliage_AdditionalMovements_04}}
 {id=WavePlayer_11, properties {soundWaveAssetPtr=/Game/Sounds/Fort_Foley_Sounds/Foliage/Foliage_AdditionalMovements_05}}
 {id=WavePlayer_12, properties {soundWaveAssetPtr=/Game/Sounds/Fort_Foley_Sounds/Foliage/Foliage_AdditionalMovements_06}}
 {id=WavePlayer_13, properties {soundWaveAssetPtr=/Game/Sounds/Fort_Foley_Sounds/Foliage/Foliage_Player_Movement_Strong_Normal_08}}
 {id=WavePlayer_14, properties {soundWaveAssetPtr=/Game/Sounds/Fort_Foley_Sounds/Foliage/Foliage_Player_Movement_Strong_Normal_09}}
 {id=WavePlayer_15, properties {soundWaveAssetPtr=/Game/Sounds/Fort_Foley_Sounds/Foliage/Foliage_Player_Movement_Strong_Normal_10}}
 {id=WavePlayer_16, properties {soundWaveAssetPtr=&Foliage_Player_Movement_Strong_Normal_01}}
 {id=WavePlayer_17, properties {soundWaveAssetPtr=&Foliage_Player_Movement_Strong_Normal_02}}
 {id=WavePlayer_18, properties {soundWaveAssetPtr=/Game/Sounds/Fort_Foley_Sounds/Foliage/Foliage_Player_Movement_Strong_Normal_03}}
 {id=WavePlayer_19, properties {soundWaveAssetPtr=/Game/Sounds/Fort_Foley_Sounds/Foliage/Foliage_Player_Movement_Strong_Normal_04}}
 {id=WavePlayer_2, properties {soundWaveAssetPtr=&Foliage_Player_Movement_Strong_Normal_03}}
 {id=WavePlayer_20, properties {soundWaveAssetPtr=/Game/Sounds/Fort_Foley_Sounds/Foliage/Foliage_Player_Movement_Strong_Normal_05}}
 {id=WavePlayer_22, properties {soundWaveAssetPtr=/Game/Sounds/Fort_Foley_Sounds/Foliage/Foliage_AdditionalMovements_08}}
 {id=WavePlayer_23, properties {soundWaveAssetPtr=/Game/Sounds/Fort_Foley_Sounds/Foliage/Foliage_AdditionalMovements_09}}
 {id=WavePlayer_25, properties {soundWaveAssetPtr=/Game/Sounds/Fort_Foley_Sounds/Foliage/Foliage_AdditionalMovements_11}}
 {id=WavePlayer_26, properties {soundWaveAssetPtr=/Game/Sounds/Fort_Foley_Sounds/Foliage/Foliage_AdditionalMovements_12}}
 {id=WavePlayer_27, properties {soundWaveAssetPtr=/Game/Sounds/Fort_Foley_Sounds/Foliage/Foliage_AdditionalMovements_13}}
 {id=WavePlayer_28, properties {soundWaveAssetPtr=/Game/Sounds/Fort_Foley_Sounds/Foliage/Foliage_AdditionalMovements_14}}
 {id=WavePlayer_29, properties {soundWaveAssetPtr=/Game/Sounds/Fort_Foley_Sounds/Foliage/Foliage_AdditionalMovements_15}}
 {id=WavePlayer_3, properties {soundWaveAssetPtr=&Foliage_Player_Movement_Strong_Normal_04}}
 {id=WavePlayer_30, properties {soundWaveAssetPtr=/Game/Sounds/Fort_Foley_Sounds/Foliage/Foliage_AdditionalMovements_16}}
 {id=WavePlayer_31, properties {soundWaveAssetPtr=/Game/Sounds/Fort_Foley_Sounds/Foliage/Foliage_AdditionalMovements_17}}
 {id=WavePlayer_32, properties {soundWaveAssetPtr=/Game/Sounds/Fort_Foley_Sounds/Foliage/Foliage_AdditionalMovements_18}}
 {id=WavePlayer_33, properties {soundWaveAssetPtr=/Game/Sounds/Fort_Foley_Sounds/Foliage/Foliage_AdditionalMovements_01}}
 {id=WavePlayer_34, properties {soundWaveAssetPtr=/Game/Sounds/Fort_Foley_Sounds/Foliage/Foliage_AdditionalMovements_02}}
 {id=WavePlayer_35, properties {soundWaveAssetPtr=/Game/Sounds/Fort_Foley_Sounds/Foliage/Foliage_AdditionalMovements_03}}
 {id=WavePlayer_36, properties {soundWaveAssetPtr=&Foliage_AdditionalMovements_04}}
 {id=WavePlayer_37, properties {soundWaveAssetPtr=&Foliage_AdditionalMovements_05}}
 {id=WavePlayer_4, properties {soundWaveAssetPtr=&Foliage_Player_Movement_Strong_Normal_05}}
 {id=WavePlayer_5, properties {soundWaveAssetPtr=/Game/Sounds/Fort_Foley_Sounds/Foliage/Foliage_Player_Movement_Strong_Normal_06}}
 {id=WavePlayer_6, properties {soundWaveAssetPtr=/Game/Sounds/Fort_Foley_Sounds/Foliage/Foliage_Player_Movement_Strong_Normal_07}}
 {id=WavePlayer_7, properties {soundWaveAssetPtr=&Foliage_AdditionalMovements_01}}
]
properties {
 maxDistance=3584
 volumeMultiplier=1.3
 concurrencyOverrides {bLimitToOwner=true, maxCount=1, volumeScaleAttackTime=0.2}
 concurrencySet [
  /Game/Sounds/Fort_Foley_Sounds/Foliage/Settings/Athena_Foley_Foliage_Local_Conc
  /Game/Sounds/Fort_Foley_Sounds/Foliage/Settings/Athena_Foley_Foliage_Global_Conc
 ]
}
